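# Niagara system for the HighTower Tapas glider, with an opening electric flare and lightning flash, then recurring bolts, crackle and glow.
# Most emitters follow sockets on the User.Weapon skeletal mesh, using mesh bolts, sprites and flickering lights driven by curves.
# Burst emitters export particle data to User.LightningAudio for sound.
# Speedlines spawn per distance moved, gated by deployment and direction dot products, scaled by speed and time of day.
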NiagaraSystem /BRCosmetics/Gadgets/Assets/VinderTech_GliderChute/Glider_HighTower_Tapas/FX/NS_Hightower_Tapas_Glider
dependencies {
 usedDataInterfaces [
  /Script/Niagara.NiagaraDataInterfaceExport
  &NiagaraDataInterfaceCurve
  &NiagaraDataInterfaceVectorCurve
  &NiagaraDataInterfaceVector2DCurve
 ]
 usedDynamicInputs [&NormalizeFloat, &VectorLength]
 usedModules [
  &SystemState
  &self:SystemSpawnScript.SystemScriptSource.SystemScriptGraph.NiagaraNodeAssignment_0.SetVariables_#1
 ]
}
emitters {
 Flare {
  emitterUpdateScript [
   &EmitterState(Life Cycle Mode=::NewEnumerator1, Loop Behavior=::NewEnumerator1, Loop Duration=0.2, UseLoopDelay=false)
   &SpawnRate(SpawnRate=0, Use Spawn Probability=false)
   &SpawnBurst_Instantaneous(Spawn Count=1, Use Spawn Probability=false)
   &SpawnBurst_Instantaneous(Spawn Count=1, Spawn Time=0.1, Use Spawn Probability=false) {module=SpawnBurst_Instantaneous001}
  ]
  particleSpawnScript [
   /Niagara/Modules/Spawn/Initialization/InitializeParticle(
    Color=<rgba 0.1929 0.8314 3 0.7>
    Lifetime=0.2
    Module.Write Color=true
    Module.Write Lifetime=true
    Module.Write Mass=true
    Module.Write Material Random=true
    Module.Write Position=false
    Module.Write Scale=false
    Module.Write SpriteRotation=false
    Module.Write SpriteSize=true
    Position=&SimulationPosition()
    Sprite Size=<300 300>
   )
   &AddVelocity(Coordinate Space=::World, Velocity=<0 0 0.1>)
  ]
  particleUpdateScript [
   &SkeletalMeshLocation(
    Inherit Velocity=false
    Interpolate Mesh Sampling=false
    MaskSpawnGroup=false
    Mesh Sampling Type=::NewEnumerator5
    Sampled Orientation Coordinate Space=::World
    Skeletal Mesh=$User.Weapon
    Socket Sampling Mode=::NewEnumerator5
   )
   &ParticleState()
   &SolveForcesAndVelocity(Clamp Velocity=false, Limit Acceleration=false)
   &ScaleColor(
    Scale Alpha=&FloatFromCurve(FloatCurve=DataInterface() {propertyValues {Curve {keys [
         {arriveTangent=-2.7664, interpMode=RCIM_Cubic, leaveTangent=-2.7664, tangentMode=RCTM_User, time=0, value=0.5}
         {arriveTangent=2, interpMode=RCIM_Cubic, leaveTangent=2, tangentMode=RCTM_Break, time=0.2, value=0.2621}
         {arriveTangent=-0.3277, interpMode=RCIM_Cubic, leaveTangent=-3.899, tangentMode=RCTM_Break, time=0.25, value=1}
         <at 1 0>
        ]}}})
    ScaleA=true
    ScaleRGB=true
   )
   &DynamicMaterialParameters(
    Index 0 Param 1=&FloatFromCurve(FloatCurve=DataInterface() {propertyValues {Curve {keys [<at 0.25 0.7>,<at 0.5 0.05>]}}})
    Index 0 Param 2=2
    Index 0 Param 3=1
    Index 0 Param 4=1
    Module.Param0WriteEnabled=true
    Module.Param1WriteEnabled=true
    Module.Param2WriteEnabled=true
    Module.Param3WriteEnabled=true
   )
   /SetVariables(Particles.LightRadius=&UniformRangedFloat(Minimum=80, Maximum=200)) {isSetParameters=true, module=SetVariables_#2}
   /SetVariables(Particles.light color:LinearColor=&MakeLinearColorFromVectorAndFloat(
     "Vector (RGB)"=&MakeVectorFromLinearColorRGB(Color=<rgba 1.1934 4.0631 20 1>)
     "Float (Alpha)"=&FloatFromCurve(FloatCurve=DataInterface() {propertyValues {Curve {keys [<at 0 1>,<at 0.0918 0.159>,<at 0.1514 0.9748>,<at 0.2547 0.1132>,<at 0.3492 0.8648>,<at 0.3825 0.1223>,<at 0.4928 0.8923>,<at 0.5611 0.1407>,<at 0.6276 0.9107>,<at 0.6959 0.1407>,<at 0.7677 0.874>,<at 0.85 0.1315>,<at 0.9235 0.874>,<at 1 0>]}}})
    )) {isSetParameters=true, module=SetVariables_#3}
   &CameraOffset(Camera Offset Amount=50)
  ]
  rendererClasses [&NiagaraSpriteRendererProperties, &NiagaraLightRendererProperties]
  renderers [
   {rendererClass=&NiagaraSpriteRendererProperties}
   {rendererClass=&NiagaraLightRendererProperties, rendererIndex=1}
  ]
 }
 FlareTop {
  emitterUpdateScript [
   &EmitterState(Life Cycle Mode=::NewEnumerator1, Loop Behavior=::NewEnumerator1, Loop Duration=0.2, UseLoopDelay=false)
   &SpawnRate(SpawnRate=5, Use Spawn Probability=false)
   &SpawnBurst_Instantaneous(Spawn Count=1, Use Spawn Probability=false)
  ]
  particleSpawnScript [
   &InitializeParticle(
    Color=<rgba 0 0.5819 5 1>
    Lifetime=0.4
    Module.Write Color=true
    Module.Write Lifetime=true
    Module.Write Mass=true
    Module.Write Material Random=true
    Module.Write Position=false
    Module.Write Scale=false
    Module.Write SpriteRotation=false
    Module.Write SpriteSize=true
    Position=&SimulationPosition()
    Sprite Size=&Vector2DFromFloat(Value=500)
   )
   &AddVelocity(Velocity=<0 0 0.1>)
  ]
  particleUpdateScript [
   &SkeletalMeshLocation(
    Inherit Velocity=false
    Interpolate Mesh Sampling=false
    MaskSpawnGroup=false
    Mesh Sampling Type=::NewEnumerator5
    Sampled Orientation Coordinate Space=::World
    Sampled Position Offset=<0 0 370>
    Skeletal Mesh=$User.Weapon
    Socket Sampling Mode=::NewEnumerator5
   )
   &ParticleState()
   &SolveForcesAndVelocity(Clamp Velocity=false, Limit Acceleration=false)
   &ScaleColor(
    Scale Alpha=&FloatFromCurve(FloatCurve=DataInterface() {propertyValues {Curve {keys [
         {arriveTangent=-2.7664, interpMode=RCIM_Cubic, leaveTangent=-2.7664, tangentMode=RCTM_User, time=0, value=0.5}
         {arriveTangent=2, interpMode=RCIM_Cubic, leaveTangent=2, tangentMode=RCTM_Break, time=0.2, value=0.2621}
         {arriveTangent=-0.3277, interpMode=RCIM_Cubic, leaveTangent=-3.899, tangentMode=RCTM_Break, time=0.25, value=1}
         <at 1 0>
        ]}}})
    ScaleA=true
    ScaleRGB=true
   )
   &DynamicMaterialParameters(
    Index 0 Param 1=&FloatFromCurve(FloatCurve=DataInterface() {propertyValues {Curve {keys [<at 0.25 0.7>,<at 0.5 0.05>]}}})
    Index 0 Param 2=2
    Index 0 Param 3=1
    Index 0 Param 4=1
    Module.Param0WriteEnabled=true
    Module.Param1WriteEnabled=true
    Module.Param2WriteEnabled=true
    Module.Param3WriteEnabled=true
   )
   &CameraOffset(Camera Offset Amount=50)
  ]
  rendererClasses [&NiagaraSpriteRendererProperties]
  renderers [{rendererClass=&NiagaraSpriteRendererProperties}]
 }
 Lightning_big02 {
  emitterUpdateScript [
   &EmitterState(Life Cycle Mode=::NewEnumerator1, Loop Behavior=::NewEnumerator1, Loop Duration=1, UseLoopDelay=false)
   &SpawnBurst_Instantaneous(Spawn Count=2, Use Spawn Probability=false)
   &SpawnBurst_Instantaneous(Spawn Count=1, Spawn Time=0.1, Use Spawn Probability=false) {module=SpawnBurst_Instantaneous001}
  ]
  particleSpawnScript [
   &InitializeParticle(
    Color=&MakeLinearColorFromVectorAndFloat("Vector (RGB)"=<1 1 1>, "Float (Alpha)"=5)
    Lifetime=&UniformRangedFloat(Minimum=0.1, Maximum=0.15)
    Mesh Scale=<0.17 0.17 0.18>
    Module.Write Color=true
    Module.Write Lifetime=true
    Module.Write Mass=true
    Module.Write Material Random=true
    Module.Write Position=false
    Module.Write Scale=true
    Module.Write SpriteRotation=false
    Module.Write SpriteSize=false
    Position=&SimulationPosition()
   )
   /Niagara/Modules/Spawn/Orientation/V2/InitialMeshOrientation(
    Orientation Vector=<0 0 -1>
    Rotation=&UniformRangedVector(Minimum=<0 0 0>, Maximum=<0 0 1>, Randomness Mode=::NewEnumerator1, Random Seed=10)
    Use Orientation Vector=true
    Use Rotation Vector=true
   )
   &AddVelocity(Coordinate Space=::World, Velocity=<0 0 1>)
   &SkeletalMeshLocation(
    Bone Sampling Mode=::NewEnumerator4
    Inherit Velocity=false
    Interpolate Mesh Sampling=false
    MaskSpawnGroup=false
    Sampled Orientation Coordinate Space=::World
    Sampled Position Offset=<0 0 20>
    Skeletal Mesh=DataInterface() {propertyValues {
      DefaultMesh=null
      ExcludeBoneName=root
      OptionalComponentClass=null
      PreviewMesh=/BRCosmetics/Gadgets/Assets/VinderTech_GliderChute/Glider_HighTower_Tapas/Meshes/Glider_HighTower_Tapas
      SkinningMode=SkinOnTheFly
      SoftSourceActor=null
      SourceMode=Default
      UvSetIndex=0
      WholeMeshLOD=-1
      bExcludeBone=false
      bReadDeformedGeometry=true
      bRequireCurrentFrameData=true
      ComponentTags []
      FilteredBones [Pickaxe_Root]
      FilteredSockets []
      MeshUserParameter {parameter {name=None, typeDefHandle {registeredTypeIndex=1}, varData []}}
      SamplingRegions []
     }}
   ) {module=SkeletalMeshLocation001}
  ]
  particleUpdateScript [
   &DynamicMaterialParameters(
    Index 0 Param 1=&UniformRangedFloat(Minimum=1, Maximum=5)
    Index 0 Param 2=&UniformRangedFloat(Minimum=0, Maximum=1)
    Index 0 Param 3=1
    Index 0 Param 4=1
    Module.Param0WriteEnabled=true
    Module.Param1WriteEnabled=true
    Module.Param2WriteEnabled=true
    Module.Param3WriteEnabled=true
   )
   &ParticleState()
   &SolveForcesAndVelocity(Clamp Velocity=false, Limit Acceleration=false)
  ]
  rendererClasses [&NiagaraMeshRendererProperties]
  renderers [{rendererClass=&NiagaraMeshRendererProperties, rendererIndex=0}]
 }
 Lightning_center {
  emitterUpdateScript [
   &EmitterState(Life Cycle Mode=::NewEnumerator1, Loop Behavior=::NewEnumerator1, Loop Duration=1, UseLoopDelay=false)
   &SpawnBurst_Instantaneous(Spawn Count=1, Use Spawn Probability=false)
   &SpawnBurst_Instantaneous(Spawn Count=1, Spawn Time=0.1, Use Spawn Probability=false) {module=SpawnBurst_Instantaneous001}
  ]
  particleSpawnScript [
   &InitializeParticle(
    Color=&MakeLinearColorFromVectorAndFloat(
     "Vector (RGB)"=<1 1 1>
     "Float (Alpha)"=&FloatFromCurve(FloatCurve=DataInterface() {propertyValues {Curve {keys [<at 0 5>,<at 1 0>]}}})
    )
    Lifetime=&UniformRangedFloat(Minimum=0.06, Maximum=0.08)
    Mesh Scale=&UniformRangedVector(Minimum=<1.2 0.6 0.6>, Maximum=<1.2 0.6 0.6>)
    Module.Write Color=true
    Module.Write Lifetime=true
    Module.Write Mass=true
    Module.Write Material Random=true
    Module.Write Position=false
    Module.Write Scale=true
    Module.Write SpriteRotation=false
    Module.Write SpriteSize=false
    Position=&SimulationPosition()
   )
   &InitialMeshOrientation(
    Rotation=&UniformRangedVector(Minimum=<0 0 0>, Maximum=<1 0 0>)
    Use Orientation Vector=true
    Use Rotation Vector=true
   )
   &DynamicMaterialParameters(
    Index 0 Param 1=20
    Index 0 Param 2=&UniformRangedFloat(Minimum=0.2, Maximum=0.2)
    Index 0 Param 3=1
    Index 0 Param 4=1
    Module.Param0WriteEnabled=true
    Module.Param1WriteEnabled=true
    Module.Param2WriteEnabled=true
    Module.Param3WriteEnabled=true
   )
   &AddVelocity(Coordinate Space=::World, Velocity=<0 0 1>)
  ]
  particleUpdateScript [
   &SkeletalMeshLocation(
    Inherit Velocity=false
    Interpolate Mesh Sampling=false
    MaskSpawnGroup=false
    Mesh Sampling Type=::NewEnumerator5
    Sampled Orientation Coordinate Space=::World
    Skeletal Mesh=$User.Weapon
    Socket Sampling Mode=::NewEnumerator5
   )
   &ParticleState()
   &SolveForcesAndVelocity(Clamp Velocity=false, Limit Acceleration=false)
   &ScaleColor(
    Scale Alpha=&FloatFromCurve(FloatCurve=DataInterface() {propertyValues {Curve {keys [<at 0 0>,<at 0.2 1>,<at 1 0>]}}})
    ScaleA=true
    ScaleRGB=true
   )
  ]
  rendererClasses [&NiagaraMeshRendererProperties]
  renderers [{rendererClass=&NiagaraMeshRendererProperties, rendererIndex=0}]
 }
 Lightning_residue {
  emitterUpdateScript [
   &EmitterState(Life Cycle Mode=::NewEnumerator1, Loop Behavior=::NewEnumerator1, Loop Duration=1, UseLoopDelay=false)
   &SpawnBurst_Instantaneous(Spawn Count=3, Spawn Time=0.18, Use Spawn Probability=false)
   &SpawnBurst_Instantaneous(Spawn Count=2, Spawn Time=0.3, Use Spawn Probability=false) {module=SpawnBurst_Instantaneous001}
  ]
  particleSpawnScript [
   &InitializeParticle(
    Color=<rgba 2.2611 16.188 20 5>
    Lifetime=&UniformRangedFloat(Minimum=0.08, Maximum=0.1)
    Mesh Scale=&UniformRangedVector(Minimum=<3 3 8>, Maximum=<8 8 8>)
    Module.Write Color=true
    Module.Write Lifetime=true
    Module.Write Mass=true
    Module.Write Material Random=true
    Module.Write Position=false
    Module.Write Scale=true
    Module.Write SpriteRotation=false
    Module.Write SpriteSize=false
    Position=&SimulationPosition()
   )
   &InitialMeshOrientation(
    Rotation=&UniformRangedVector(Minimum=<0 0 0>, Maximum=<1 0 0>)
    Use Orientation Vector=true
    Use Rotation Vector=true
   ) {module=InitialMeshOrientation001}
   &DynamicMaterialParameters(
    Index 0 Param 1=&UniformRangedFloat(Minimum=5, Maximum=10)
    Index 0 Param 3=&UniformRangedFloat(Minimum=0.2, Maximum=1)
    Index 0 Param 4=&UniformRangedFloat(Minimum=0, Maximum=1)
    Module.Param0WriteEnabled=true
    Module.Param1WriteEnabled=false
    Module.Param2WriteEnabled=true
    Module.Param3WriteEnabled=true
   )
   /Niagara/Modules/Spawn/Location/SphereLocation(MaskSpawnGroup=false, Module.Override Randomness=false, Sphere Radius=10, Surface Only=false)
   &AddVelocity(Coordinate Space=::World, Velocity=<0 0 1>)
  ]
  particleUpdateScript [
   &SkeletalMeshLocation(
    Inherit Velocity=false
    Interpolate Mesh Sampling=false
    MaskSpawnGroup=false
    Mesh Sampling Type=::NewEnumerator5
    Sampled Orientation Coordinate Space=::World
    Skeletal Mesh=$User.Weapon
    Socket Sampling Mode=::NewEnumerator5
   )
   &ParticleState()
   &SolveForcesAndVelocity(Clamp Velocity=false, Limit Acceleration=false)
   &ScaleColor(
    Scale Alpha=&FloatFromCurve(
     FloatCurve=DataInterface() {propertyValues {Curve {keys [<at 0 1 RCIM_Constant>,<at 0.1189 0 RCIM_Constant>,<at 0.2516 0.997 RCIM_Constant>,<at 0.6735 -0.0069 RCIM_Constant>,<at 0.7649 1.0117 RCIM_Constant>,<at 1 0 RCIM_Constant>]}}}
     Scale Curve=&Multiply_Float(
      A=&FloatFromCurve(FloatCurve=DataInterface() {propertyValues {Curve {keys [<at 0 1 RCIM_Constant>,<at 0.1561 -0.0068 RCIM_Constant>,<at 0.3011 0.9877 RCIM_Constant>,<at 0.6424 0.0069 RCIM_Constant>,<at 0.7618 0.9877 RCIM_Constant>,<at 1 0 RCIM_Constant>]}}})
      B=&UniformRangedFloat(Minimum=1, Maximum=1)
     )
    )
    Scale RGB=&VectorFromFloat(Value=&UniformRangedFloat(Minimum=1, Maximum=5))
    ScaleA=true
    ScaleRGB=true
   ) {module=ScaleColor001}
   &DynamicMaterialParameters(
    Index 0 Param 2=&FloatFromCurve(FloatCurve=DataInterface() {propertyValues {Curve {keys [{interpMode=RCIM_Cubic, leaveTangent=-2.846, tangentMode=RCTM_Break, time=0, value=1}, <at 1 -0.25>]}}})
    Module.Param0WriteEnabled=false
    Module.Param1WriteEnabled=true
    Module.Param2WriteEnabled=false
    Module.Param3WriteEnabled=false
   ) {module=DynamicMaterialParameters001}
  ]
  rendererClasses [&NiagaraMeshRendererProperties]
  renderers [{rendererClass=&NiagaraMeshRendererProperties, rendererIndex=0}]
 }
 Lightning_small {
  emitterUpdateScript [
   &EmitterState(Inactive Response=::NewEnumerator1, Life Cycle Mode=::NewEnumerator1, Loop Duration=1, UseLoopDelay=false)
   &SpawnRate(SpawnRate=45, Use Spawn Probability=false)
  ]
  particleSpawnScript [
   &InitializeParticle(
    Color=<rgba 0.4098 5.5672 8 1>
    Lifetime=&UniformRangedFloat(Minimum=0.1, Maximum=0.1)
    Mesh Scale=&Multiply_Vector(
     A=&UniformRangedVector(Minimum=<0.5 0.7 0.7>, Maximum=<0.5 0.5 0.5>)
     B=&VectorFromFloat(Value=1.5)
    )
    Module.Write Color=true
    Module.Write Lifetime=true
    Module.Write Mass=true
    Module.Write Material Random=true
    Module.Write Position=true
    Module.Write Scale=true
    Module.Write SpriteRotation=false
    Module.Write SpriteSize=false
    Position=&SimulationPosition()
    Sprite Size=<25 25>
   )
   &InitialMeshOrientation(
    Rotation=&UniformRangedVector(Minimum=<-1 0 0>, Maximum=<1 0 0>)
    Use Orientation Vector=true
    Use Rotation Vector=true
   )
   &SphereLocation(
    MaskSpawnGroup=false
    Module.Override Randomness=false
    Offset=&UniformRangedVector(Minimum=<0 0 0>, Maximum=<5 5 5>)
    Sphere Origin:NiagaraPosition=&MakeVector(X=0, Y=0, Z=0)
    Sphere Radius=16
    Surface Only=false
   )
   &AddVelocityFromPoint(Module.Use Falloff Distance=false, Velocity Strength=0.01)
  ]
  particleUpdateScript [
   &SkeletalMeshLocation(
    Inherit Velocity=false
    Interpolate Mesh Sampling=false
    MaskSpawnGroup=false
    Mesh Sampling Type=::NewEnumerator5
    Sampled Position Offset:NiagaraPosition=$Particles.Initial.Position
    Skeletal Mesh=$User.Weapon
    Socket Sampling Mode=::NewEnumerator5
   )
   &SolveForcesAndVelocity(Clamp Velocity=false, Limit Acceleration=false)
   &ParticleState()
   &ScaleColor(Scale RGB=<10 10 10>, ScaleA=true, ScaleRGB=true)
   &DynamicMaterialParameters(
    Index 0 Param 1=&UniformRangedFloat(Minimum=4, Maximum=7)
    Index 0 Param 2=&FloatFromCurve(FloatCurve=DataInterface() {propertyValues {Curve {keys [<at 0.1129 1>,<at 1 -1>]}}})
    Index 0 Param 3=&UniformRangedFloat(Minimum=0.9, Maximum=1.5)
    Module.Param0WriteEnabled=true
    Module.Param1WriteEnabled=true
    Module.Param2WriteEnabled=true
    Module.Param3WriteEnabled=true
   )
  ]
  rendererClasses [&NiagaraMeshRendererProperties]
  renderers [{rendererClass=&NiagaraMeshRendererProperties, rendererIndex=0}]
 }
 Speedlines {
  emitterUpdateScript [
   &EmitterState(
    Delay First Loop Only=true
    Enable Distance Culling=true
    Inactive Response=::NewEnumerator1
    Life Cycle Mode=::NewEnumerator1
    Loop Delay=1
    Loop Duration=1
    MaxDistance=2048
    Module.bUseMaxDistance=true
    Module.bUseMinDistance=false
    Scalability Mode=::NewEnumerator1
    UseLoopDelay=true
   )
   &SpawnPerUnit(
    Module.Use Max Movement Threshold=true
    Module.Use Movement Tolerance=false
    Spawn Per Unit=10
    Spawn Probability:NiagaraFloat=&MakeCustomFloatFromBool(
     Bool=&Boolean_AndOperation(
      Bool1=$User.bIsFullyDeployed
      Bool2=&Boolean_AndOperation(
       Bool1=&SetBoolByFloatComparison(A=$User.ForwardDot, B=0.2, Comparison Type=::NewEnumerator2)
       Bool2=&Boolean_AndOperation(
        Bool1=&SetBoolByFloatComparison(A=$User.RightDot, B=-0.5)
        Bool2=&SetBoolByFloatComparison(A=$User.RightDot, B=0.5, Comparison Type=::NewEnumerator6)
       )
      )
     )
     True Float=1
     False Float=0
    )
    Use Spawn Probability=true
   )
  ]
  particleSpawnScript [
   &InitializeParticle(
    Color=&MakeLinearColorFromVectorAndFloat(
     "Vector (RGB)"=<1 1 1>
     "Float (Alpha)"=&Multiply_Float(A=&UniformRangedFloat(Minimum=0.2, Maximum=0.3), B=0.5)
    )
    Lifetime=&UniformRangedFloat(Minimum=0.1, Maximum=0.15)
    Module.Write Color=true
    Module.Write Lifetime=true
    Module.Write Mass=true
    Module.Write Material Random=true
    Module.Write Position=false
    Module.Write Scale=false
    Module.Write SpriteRotation=false
    Module.Write SpriteSize=true
    Position=&SimulationPosition()
    Sprite Size=&UniformRangedVector2D(Minimum=<4 16>, Maximum=<6 24>)
   )
   &SkeletalMeshLocation(
    Interpolate Mesh Sampling=false
    MaskSpawnGroup=false
    Mesh Sampling Type=::NewEnumerator5
    Orientation Sampling=::NewEnumerator1
    Skeletal Mesh=DataInterface() {propertyValues {
      DefaultMesh=null
      ExcludeBoneName=root
      OptionalComponentClass=null
      PreviewMesh=&Glider_HighTower_Tapas
      SkinningMode=SkinOnTheFly
      SoftSourceActor=null
      SourceMode=Default
      UvSetIndex=0
      WholeMeshLOD=-1
      bExcludeBone=false
      bReadDeformedGeometry=true
      bRequireCurrentFrameData=true
      ComponentTags []
      FilteredBones [Glider_Base]
      FilteredSockets [Lightning_FX]
      MeshUserParameter {parameter {name=None, typeDefHandle {registeredTypeIndex=1}, varData []}}
      SamplingRegions []
     }}
    Socket Sampling Mode=::NewEnumerator5
   )
   /Niagara/Modules/Spawn/Location/V2/SphereLocation(
    MaskSpawnGroup=false
    Offset=<0 -50 0>
    Override Seed=false
    Sphere Radius=200
    Surface Only=false
    UseNonUniformScale=false
    UseOffset=true
    UseSphereOrientation=false
   )
   &AddVelocityFromPoint(
    Module.Use Falloff Distance=false
    Origin Offset=&Add_Vector(A=Linked() {linkedVariable {name=Output.SkeletalMeshLocation.SampledPositionWithoutOffset, type=NiagaraPosition}}, B=<50 -50 0>)
    Velocity Strength=1000
   )
   &AddVelocity(Velocity=&RotateVectorByQuaternion(Vector To Rotate=<0 0 -10>, Quaternion=$Particles.SkeletalMeshLocation.SampledRotation)) {enabled=false}
   &CameraOffset(Camera Offset Amount=-20)
  ]
  particleUpdateScript [
   &ParticleState()
   &SolveForcesAndVelocity(Clamp Velocity=false, Limit Acceleration=false)
   &ScaleSpriteSize(Scale Factor=&Vector2DFromCurve(
     Vector2Curve=DataInterface() {propertyValues {XCurve {keys [<at 0 0.5 RCIM_Cubic>,<at 0.5 1 RCIM_Cubic>]}, YCurve {keys [<at 0 0 RCIM_Cubic>,<at 0.5 1 RCIM_Cubic>]}}}
     ScaleCurve=&Vector2DFromFloat(Value=1)
    ))
   &ScaleSpriteSizeBySpeed(Max Scale Factor=<1 20>, Min Scale Factor=<0.01 0.05>, Velocity Threshold=1)
   &ScaleColor(
    Scale Alpha=&Multiply_Float(
     A=&FloatFromCurve(
      FloatCurve=DataInterface() {propertyValues {Curve {keys [
          <at 0 0 RCIM_Cubic>
          <at 0.3759 1.0207>
          {arriveTangent=-2.221, interpMode=RCIM_Cubic, leaveTangent=-2.221, tangentMode=RCTM_User, time=1, value=0}
         ]}}}
      Scale Curve=$User.SpeedLine
     )
     B=&FloatFromCurve(FloatCurve=DataInterface() {propertyValues {Curve {keys [<at 0 0>,<at 1 1>]}}}, CurveIndex=$System.PlayerSpeed)
    )
    Scale RGB=&VectorFromFloat(Value=&ClampFloat(
      Float=$User.ForwardDot
      Max=&FNTimeOfDayScalar(Night Scale=1, Morning Scale=0.3, Evening Scale=0.5)
     ))
    ScaleA=true
    ScaleRGB=true
   )
  ]
  rendererClasses [&NiagaraSpriteRendererProperties]
  renderers [{rendererClass=&NiagaraSpriteRendererProperties}]
 }
 glow {
  emitterUpdateScript [
   &EmitterState(
    Delay First Loop Only=true
    Inactive Response=::NewEnumerator1
    Life Cycle Mode=::NewEnumerator1
    Loop Delay=0.3
    Loop Duration=0.2
    UseLoopDelay=true
   )
   &SpawnRate(SpawnRate=5, Use Spawn Probability=false)
   &SpawnBurst_Instantaneous(Spawn Count=1, Use Spawn Probability=false)
  ]
  particleSpawnScript [
   &InitializeParticle(
    Color=<rgba 0.0156 0.7396 3 0.5>
    Lifetime=0.8
    Mesh Scale=&Multiply_Vector(
     A=<3 2.5 2.5>
     B=&VectorFromFloat(Value=&UniformRangedFloat(Minimum=0.5, Maximum=1))
    )
    Module.Write Color=true
    Module.Write Lifetime=true
    Module.Write Mass=true
    Module.Write Material Random=true
    Module.Write Position=false
    Module.Write Scale=false
    Module.Write SpriteRotation=true
    Module.Write SpriteSize=true
    Position=&SimulationPosition()
    Sprite Rotation=&UniformRangedFloat(Minimum=0, Maximum=1)
    Sprite Size=<320 320>
   )
   &DynamicMaterialParameters(
    Index 0 Param 2=2
    Index 0 Param 3=1
    Index 0 Param 4=1
    Module.Param0WriteEnabled=false
    Module.Param1WriteEnabled=true
    Module.Param2WriteEnabled=true
    Module.Param3WriteEnabled=true
   )
  ]
  particleUpdateScript [
   &SkeletalMeshLocation(
    Inherit Velocity=false
    Interpolate Mesh Sampling=false
    MaskSpawnGroup=false
    Mesh Sampling Type=::NewEnumerator5
    Skeletal Mesh=$User.Weapon
    Socket Sampling Mode=::NewEnumerator5
   )
   &ParticleState()
   &SolveForcesAndVelocity(Clamp Velocity=false, Limit Acceleration=false)
   &DynamicMaterialParameters(
    Index 0 Param 1=&FloatFromCurve(FloatCurve=DataInterface() {propertyValues {Curve {keys [<at 0.25 0.7>,<at 0.5 0.05>,<at 1 0>]}}})
    Module.Param0WriteEnabled=true
    Module.Param1WriteEnabled=false
    Module.Param2WriteEnabled=false
    Module.Param3WriteEnabled=false
   ) {module=DynamicMaterialParameters002}
   &ScaleColor(
    Scale Alpha=&FloatFromCurve(FloatCurve=DataInterface() {propertyValues {Curve {keys [<at 0 0.5>,<at 0.2 0.262>,<at 0.25 1>,<at 1 0>]}}}, Scale Curve=0.1)
    ScaleA=true
    ScaleRGB=true
   )
   /SetVariables(Particles.light color:LinearColor=&MakeLinearColorFromVectorAndFloat(
     "Vector (RGB)"=&MakeVectorFromLinearColorRGB(Color=<rgba 0.4843 3.3558 10 1>)
     "Float (Alpha)"=&FloatFromCurve(FloatCurve=DataInterface() {propertyValues {Curve {keys [<at 0 1>,<at 0.0918 0.159>,<at 0.1514 0.9748>,<at 0.2547 0.1132>,<at 0.3492 0.8648>,<at 0.3825 0.1223>,<at 0.4928 0.8923>,<at 0.5611 0.1407>,<at 0.6276 0.9107>,<at 0.6959 0.1407>,<at 0.7677 0.874>,<at 0.85 0.1315>,<at 0.9235 0.874>,<at 1 0>]}}})
    )) {isSetParameters=true, module=SetVariables_#3}
   /SetVariables(Particles.LightRadius=&UniformRangedFloat(Minimum=10, Maximum=100)) {isSetParameters=true, module=SetVariables_#2}
  ]
  rendererClasses [&NiagaraSpriteRendererProperties, &NiagaraLightRendererProperties]
  renderers [
   {rendererClass=&NiagaraSpriteRendererProperties}
   {rendererClass=&NiagaraLightRendererProperties, rendererIndex=1}
  ]
 }
 lightning_burst {
  emitterUpdateScript [
   &EmitterState(
    Inactive Response=::NewEnumerator1
    Life Cycle Mode=::NewEnumerator1
    Loop Delay=&UniformRangedFloat(Minimum=1.1, Maximum=1.5)
    Loop Duration=&UniformRangedFloat(Minimum=0.5, Maximum=0.8)
    Recalculate Duration Each Loop=true
    UseLoopDelay=true
   )
   &SpawnRate(SpawnRate=2, Use Spawn Probability=false)
   &SpawnBurst_Instantaneous(Spawn Count=4, Use Spawn Probability=false)
  ]
  particleSpawnScript [
   &InitializeParticle(
    Lifetime=&UniformRangedFloat(Minimum=0.1, Maximum=0.2)
    Mesh Scale=&UniformRangedVector(Minimum=<0.1 0.1 0.1>, Maximum=<0.4 0.4 0.4>)
    Module.Write Color=true
    Module.Write Lifetime=true
    Module.Write Mass=true
    Module.Write Material Random=true
    Module.Write Position=false
    Module.Write Scale=true
    Module.Write SpriteRotation=false
    Module.Write SpriteSize=false
    Position=&SimulationPosition()
    Sprite Size=&UniformRangedVector2D(Minimum=<0 0>, Maximum=<1 1>)
   )
   &InitialMeshOrientation(
    Orientation Vector=<0 1 0>
    Rotation=&UniformRangedVector(Minimum=<0 -1 0>, Maximum=<0 1 0>)
    Use Orientation Vector=true
    Use Rotation Vector=true
   )
   &DynamicMaterialParameters(
    Index 0 Param 2=&UniformRangedFloat(Minimum=0.3, Maximum=0.4)
    Module.Param0WriteEnabled=false
    Module.Param1WriteEnabled=true
    Module.Param2WriteEnabled=false
    Module.Param3WriteEnabled=false
   ) {module=DynamicMaterialParameters001}
   &self:HandleLightningAudio(
    New Export particle data=DataInterface() {propertyValues {
      GPUAllocationFixedSize=64
      GPUAllocationMode=FixedSize
      GPUAllocationPerParticleSize=1
      CallbackHandlerParameter {parameter {name=User.LightningAudio, typeDefHandle {registeredTypeIndex=1}, varData []}}
     }}
    Store Data=true
   ) {module=HandleLightningAudio}
  ]
  particleUpdateScript [
   &SkeletalMeshLocation(
    Inherit Velocity=false
    Interpolate Mesh Sampling=false
    MaskSpawnGroup=false
    Mesh Sampling Type=::NewEnumerator5
    Sampled Orientation Coordinate Space=::World
    Skeletal Mesh=$User.Weapon
    Socket Sampling Mode=::NewEnumerator5
   )
   &ParticleState()
   &SolveForcesAndVelocity(Clamp Velocity=false, Limit Acceleration=false)
   &ScaleColor(
    Scale Alpha=&FloatFromCurve(FloatCurve=DataInterface() {propertyValues {Curve {keys [<at 0.15 1 RCIM_Cubic>, {arriveTangent=-1.1765, interpMode=RCIM_Cubic, leaveTangent=-1.1765, time=0.796, value=0.245}, <at 1 0>]}}})
    Scale RGB=&VectorFromFloat(Value=&FloatFromCurve(FloatCurve=DataInterface() {propertyValues {Curve {keys [<at 0 1>,<at 0.1242 0.9958 RCIM_Constant>,<at 0.1835 0 RCIM_Constant>,<at 0.284 0.984 RCIM_Constant>,<at 0.5317 0 RCIM_Constant>,<at 0.6402 0.984 RCIM_Constant>,<at 1 0>]}}}))
    ScaleA=true
    ScaleRGB=true
   ) {module=ScaleColor003}
   &DynamicMaterialParameters(
    Index 0 Param 1=&FloatFromCurve(FloatCurve=DataInterface() {propertyValues {Curve {keys [<at 0 1>,<at 0.7 0.35>,<at 1 0>]}}})
    Index 0 Param 2=0.3
    Index 0 Param 3=&FloatFromCurve(FloatCurve=DataInterface() {propertyValues {Curve {keys [<at 0.6 1>,<at 1 0>]}}}, Scale Curve=0.5)
    Index 0 Param 4=1
    Module.Param0WriteEnabled=true
    Module.Param1WriteEnabled=false
    Module.Param2WriteEnabled=true
    Module.Param3WriteEnabled=true
   )
  ]
  rendererClasses [&NiagaraMeshRendererProperties]
  renderers [{rendererClass=&NiagaraMeshRendererProperties, rendererIndex=0}]
 }
 lightning_burst02 {
  emitterUpdateScript [
   &EmitterState(
    Inactive Response=::NewEnumerator1
    Life Cycle Mode=::NewEnumerator1
    Loop Delay=&UniformRangedFloat(Minimum=1.1, Maximum=1.5)
    Loop Duration=&UniformRangedFloat(Minimum=0.5, Maximum=0.8)
    Recalculate Duration Each Loop=true
    UseLoopDelay=true
   )
   &SpawnRate(SpawnRate=2, Use Spawn Probability=false)
   &SpawnBurst_Instantaneous(Spawn Count=4, Use Spawn Probability=false)
  ]
  particleSpawnScript [
   &InitializeParticle(
    Lifetime=&UniformRangedFloat(Minimum=0.1, Maximum=0.2)
    Mesh Scale=&UniformRangedVector(Minimum=<0.2 0.5 0.5>, Maximum=<0.3 0.4 0.4>)
    Module.Write Color=true
    Module.Write Lifetime=true
    Module.Write Mass=true
    Module.Write Material Random=true
    Module.Write Position=false
    Module.Write Scale=true
    Module.Write SpriteRotation=false
    Module.Write SpriteSize=false
    Position=&SimulationPosition()
    Sprite Size=&UniformRangedVector2D(Minimum=<0 0>, Maximum=<1 1>)
   )
   &InitialMeshOrientation(
    Orientation Vector=<1 1 1>
    Rotation=&UniformRangedVector(Minimum=<-1 -1 -1>, Maximum=<1 1 1>)
    Use Orientation Vector=true
    Use Rotation Vector=true
   )
   &DynamicMaterialParameters(
    Index 0 Param 2=&UniformRangedFloat(Minimum=0.03, Maximum=0.04)
    Module.Param0WriteEnabled=false
    Module.Param1WriteEnabled=true
    Module.Param2WriteEnabled=false
    Module.Param3WriteEnabled=false
   ) {module=DynamicMaterialParameters001}
   &self:HandleLightningAudio(
    New Export particle data=DataInterface() {propertyValues {
      GPUAllocationFixedSize=64
      GPUAllocationMode=FixedSize
      GPUAllocationPerParticleSize=1
      CallbackHandlerParameter {parameter {name=User.LightningAudio, typeDefHandle {registeredTypeIndex=1}, varData []}}
     }}
    Store Data=true
   ) {module=HandleLightningAudio}
  ]
  particleUpdateScript [
   &SkeletalMeshLocation(
    Inherit Velocity=false
    Interpolate Mesh Sampling=false
    MaskSpawnGroup=false
    Mesh Sampling Type=::NewEnumerator5
    Sampled Orientation Coordinate Space=::World
    Skeletal Mesh=$User.Weapon
    Socket Sampling Mode=::NewEnumerator5
   )
   &ParticleState()
   &SolveForcesAndVelocity(Clamp Velocity=false, Limit Acceleration=false)
   &ScaleColor(
    Scale Alpha=&FloatFromCurve(FloatCurve=DataInterface() {propertyValues {Curve {keys [<at 0.15 1 RCIM_Cubic>, {arriveTangent=-1.1765, interpMode=RCIM_Cubic, leaveTangent=-1.1765, time=0.796, value=0.245}, <at 1 0>]}}})
    Scale RGB=&VectorFromFloat(Value=&FloatFromCurve(FloatCurve=DataInterface() {propertyValues {Curve {keys [<at 0 1>,<at 0.1242 0.9958 RCIM_Constant>,<at 0.1835 0 RCIM_Constant>,<at 0.284 0.984 RCIM_Constant>,<at 0.5317 0 RCIM_Constant>,<at 0.6402 0.984 RCIM_Constant>,<at 1 0>]}}}))
    ScaleA=true
    ScaleRGB=true
   ) {module=ScaleColor003}
   &DynamicMaterialParameters(
    Index 0 Param 1=&FloatFromCurve(FloatCurve=DataInterface() {propertyValues {Curve {keys [<at 0 1>,<at 0.7 0.35>,<at 1 0>]}}})
    Index 0 Param 2=0.05
    Index 0 Param 3=&FloatFromCurve(FloatCurve=DataInterface() {propertyValues {Curve {keys [<at 0.6 1>,<at 1 1>]}}}, Scale Curve=0.5)
    Index 0 Param 4=1
    Module.Param0WriteEnabled=true
    Module.Param1WriteEnabled=false
    Module.Param2WriteEnabled=true
    Module.Param3WriteEnabled=true
   )
  ]
  rendererClasses [&NiagaraMeshRendererProperties]
  renderers [{rendererClass=&NiagaraMeshRendererProperties, rendererIndex=0}]
 }
 lightning_burstBig {
  emitterUpdateScript [
   &EmitterState(
    Inactive Response=::NewEnumerator1
    Life Cycle Mode=::NewEnumerator1
    Loop Delay=&UniformRangedFloat(Minimum=1.8, Maximum=1.8)
    Loop Duration=1
    Recalculate Duration Each Loop=true
    UseLoopDelay=true
   )
   &SpawnBurst_Instantaneous(Spawn Count=1, Use Spawn Probability=false)
   &SpawnRate(SpawnRate=0, Use Spawn Probability=false)
   &SpawnBurst_Instantaneous(Spawn Count=1, Spawn Time=0.1, Use Spawn Probability=false) {module=SpawnBurst_Instantaneous001}
  ]
  particleSpawnScript [
   &InitializeParticle(
    Lifetime=0.3
    Mesh Scale=&UniformRangedVector(Minimum=<0.15 0.15 0.15>, Maximum=<0.2 0.15 0.15>)
    Module.Write Color=true
    Module.Write Lifetime=true
    Module.Write Mass=true
    Module.Write Material Random=true
    Module.Write Position=false
    Module.Write Scale=true
    Module.Write SpriteRotation=false
    Module.Write SpriteSize=false
    Position=&SimulationPosition()
    Sprite Size=&UniformRangedVector2D(Minimum=<0 0>, Maximum=<1 1>)
   )
   &InitialMeshOrientation(
    Orientation Vector=<0 0 1>
    Rotation=&UniformRangedVector(Minimum=<0 0 0>, Maximum=<0 0 1>)
    Use Orientation Vector=true
    Use Rotation Vector=true
   )
   &self:HandleLightningAudio(
    New Export particle data=DataInterface() {propertyValues {
      GPUAllocationFixedSize=64
      GPUAllocationMode=FixedSize
      GPUAllocationPerParticleSize=1
      CallbackHandlerParameter {parameter {name=User.LightningAudio, typeDefHandle {registeredTypeIndex=1}, varData []}}
     }}
    Store Data=true
   ) {module=HandleLightningAudio}
  ]
  particleUpdateScript [
   &SkeletalMeshLocation(
    Inherit Velocity=false
    Interpolate Mesh Sampling=false
    MaskSpawnGroup=false
    Mesh Sampling Type=::NewEnumerator5
    Sampled Orientation Coordinate Space=::World
    Skeletal Mesh=$User.Weapon
    Socket Sampling Mode=::NewEnumerator5
   )
   &ParticleState()
   &SolveForcesAndVelocity(Clamp Velocity=false, Limit Acceleration=false)
   &DynamicMaterialParameters(
    Index 0 Param 1=&FloatFromCurve(FloatCurve=DataInterface() {propertyValues {Curve {keys [<at 0 1 RCIM_Cubic>, {arriveTangent=-1, interpMode=RCIM_Cubic, leaveTangent=-1, time=0.3, value=0.5}, <at 1 0>]}}})
    Index 0 Param 2=&FloatFromCurve(FloatCurve=DataInterface() {propertyValues {Curve {keys [
         {interpMode=RCIM_Cubic, leaveTangent=-0.9598, tangentMode=RCTM_Break, time=0, value=0.5}
         {arriveTangent=5, interpMode=RCIM_Cubic, leaveTangent=5, tangentMode=RCTM_Break, time=0.25, value=0.7}
         <at 0.3 2>
         {arriveTangent=-5.439, interpMode=RCIM_Cubic, leaveTangent=-5.439, tangentMode=RCTM_User, time=0.35, value=0.6}
         {arriveTangent=-0.3966, interpMode=RCIM_Cubic, tangentMode=RCTM_Break, time=1, value=0.05}
        ]}}})
    Index 0 Param 3=&FloatFromCurve(FloatCurve=DataInterface() {propertyValues {Curve {keys [
         <at 0.1 0.5 RCIM_Cubic>
         <at 0.296 1>
         {arriveTangent=-2.9528, interpMode=RCIM_Cubic, leaveTangent=-2.9528, time=0.35, value=0.4}
         {arriveTangent=4.0435, interpMode=RCIM_Cubic, leaveTangent=4.0435, time=0.55, value=0.25}
         <at 0.993 3>
        ]}}})
    Index 0 Param 4=1
    Module.Param0WriteEnabled=true
    Module.Param1WriteEnabled=true
    Module.Param2WriteEnabled=true
    Module.Param3WriteEnabled=true
   )
   &ScaleMeshSize(Scale Factor=/Niagara/DynamicInputs/ScaleByCurve/ScaleVectorByCurve(
     InitialValue:Vector3f=<1 1 1>
     ScaleCurve:NiagaraDataInterfaceVectorCurve=DataInterface() {propertyValues {XCurve {keys [<at 1 1>]}, YCurve {keys [<at 0.3 1.25>,<at 0.6 1.5>,<at 1 0.5>]}, ZCurve {keys [<at 0.3 1.25>,<at 0.6 1.5>,<at 1 0.5>]}}}
     CurveIndex:NiagaraFloat=$Particles.NormalizedAge
    ))
   &ScaleColor(
    Scale Alpha=&FloatFromCurve(FloatCurve=DataInterface() {propertyValues {Curve {keys [<at 0.15 1 RCIM_Cubic>, {arriveTangent=-1.1765, interpMode=RCIM_Cubic, leaveTangent=-1.1765, time=0.796, value=0.245}, <at 1 0>]}}})
    Scale RGB=&VectorFromFloat(Value=&FloatFromCurve(FloatCurve=DataInterface() {propertyValues {Curve {keys [<at 0 1>,<at 0.1242 0.9958 RCIM_Constant>,<at 0.1835 0 RCIM_Constant>,<at 0.2506 0.9839 RCIM_Constant>,<at 0.5317 0 RCIM_Constant>,<at 0.6151 1 RCIM_Constant>,<at 1 0>]}}}))
    ScaleA=true
    ScaleRGB=true
   ) {module=ScaleColor002}
   /SetVariables(Particles.light color:LinearColor=&MakeLinearColorFromVectorAndFloat(
     "Vector (RGB)"=&MakeVectorFromLinearColorRGB(Color=<rgba 3.4841 29.7233 100 1>)
     "Float (Alpha)"=&FloatFromCurve(FloatCurve=DataInterface() {propertyValues {Curve {keys [<at 0 1>,<at 0.0918 0.159>,<at 0.1514 0.9748>,<at 0.2547 0.1132>,<at 0.3492 0.8648>,<at 0.3825 0.1223>,<at 0.4928 0.8923>,<at 0.5611 0.1407>,<at 0.6276 0.9107>,<at 0.6959 0.1407>,<at 0.7677 0.874>,<at 0.85 0.1315>,<at 0.9235 0.874>,<at 1 0>]}}})
    )) {isSetParameters=true, module=SetVariables_#3}
   /SetVariables(Particles.LightRadius=&UniformRangedFloat(Minimum=0, Maximum=100)) {isSetParameters=true, module=SetVariables_#2}
  ]
  rendererClasses [&NiagaraMeshRendererProperties, &NiagaraLightRendererProperties]
  renderers [
   {rendererClass=&NiagaraMeshRendererProperties, rendererIndex=0}
   {rendererClass=&NiagaraLightRendererProperties, rendererIndex=1}
  ]
 }
}
summary {
 systemName=NS_Hightower_Tapas_Glider
 emitters [
  {emitterName=lightning_burst, rendererClasses [&NiagaraMeshRendererProperties]}
  {
   emitterName=lightning_burstBig
   rendererClasses [&NiagaraMeshRendererProperties, &NiagaraLightRendererProperties]
  }
  {emitterName=Lightning_small, rendererClasses [&NiagaraMeshRendererProperties]}
  {emitterName=glow, rendererClasses [&NiagaraSpriteRendererProperties, &NiagaraLightRendererProperties]}
  {emitterName=Lightning_center, rendererClasses [&NiagaraMeshRendererProperties]}
  {emitterName=Lightning_big02, rendererClasses [&NiagaraMeshRendererProperties]}
  {emitterName=Lightning_residue, rendererClasses [&NiagaraMeshRendererProperties]}
  {emitterName=Flare, rendererClasses [&NiagaraSpriteRendererProperties, &NiagaraLightRendererProperties]}
  {emitterName=FlareTop, rendererClasses [&NiagaraSpriteRendererProperties]}
  {emitterName=lightning_burst02, rendererClasses [&NiagaraMeshRendererProperties]}
  {emitterName=Speedlines, rendererClasses [&NiagaraSpriteRendererProperties]}
 ]
 userVariables [
  User.bIsFullyDeployed {type=NiagaraBool}
  User.ForwardDot {type=NiagaraFloat, defaultValue=0, description=""}
  User.LightningAudio {
   type=Object
   description=""
   defaultValue {struct=NiagaraExt_VariableValue_Object, value {object=null, objectClass=&Object}}
  }
  User.RightDot {type=NiagaraFloat, defaultValue=0, description=""}
  User.SpeedLine {type=NiagaraFloat, defaultValue=1, description=""}
  User.Weapon {
   type=NiagaraDataInterfaceSkeletalMesh
   defaultValue {
    struct=NiagaraExt_VariableValue_DataInterface
    value {dataInterface=&self:NiagaraDataInterfaceSkeletalMesh_1}
   }
  }
 ]
}
